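# Niagara system for the PlankCoverAge emote's fiery exhaust, positioned on a skeletal mesh.
# Flame_Launch and Flash mesh emitters spawn while moving or in front end, alongside a flickering light.
# GPU ember sprites spawn by burst, rate and distance travelled, with curl noise, drag and speed-based stretching.
# System timelines scale embers and gameplay light brightness, while time-of-day curves adjust alpha.
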NiagaraSystem /BRCosmetics/Animation/Game/MainPlayer/Emotes/PlankCoverAge/FX/NS_Emote_PlankCoverAge_Exhaust
dependencies {
 usedDataInterfaces [
  &NiagaraDataInterfaceVectorCurve
  &NiagaraDataInterfaceCurve
  &NiagaraDataInterfaceColorCurve
  &NiagaraDataInterfaceMeshRendererInfo
  &NiagaraDataInterfaceSkeletalMesh
  &NiagaraDataInterfaceArrayFloat
 ]
 usedDynamicInputs [
  &MakeCustomFloatFromBool
  &SetBoolByFloatComparison
  &VectorLength
 ]
 usedModules [
  &self:SystemSpawnScript.SystemScriptSource.SystemScriptGraph.NiagaraNodeAssignment_0.SetVariables_#1
  &SystemState
  &self:SystemSpawnScript.SystemScriptSource.SystemScriptGraph.NiagaraNodeAssignment_3.SetVariables_#2
  &Timeline
 ]
}
emitters {
 EmbersGPUHigh {
  simTarget=GPUComputeSim
  emitterUpdateScript [
   &EmitterState()
   &SpawnBurst_Instantaneous(Spawn Count=55, Use Spawn Probability=false, UseLoopCountLimit=true)
   &SpawnPerUnit(
    Module.Use Max Movement Threshold=true
    Module.Use Movement Tolerance=false
    Movement Tolerance=15
    Spawn Spacing=40
    Use Spawn Probability=false
   )
   &SpawnRate(Spawn Probability=0.5, Use Spawn Probability=true) {module=SpawnRate002}
  ]
  particleSpawnScript [
   /SetVariables(Particles.PerParticleRandom:NiagaraFloat=&RandomRangeFloat(Minimum=0, Maximum=1)) {isSetParameters=true}
   &InitializeParticle(
    Color Mode=::NewEnumerator0
    Lifetime=&Multiply_Float(
     A=&Lerp_Float(A=0.999, B=0.666, Alpha=$Particles.PerParticleRandom)
     B=&Lerp_Float(A=0.5, B=1, Alpha=$System.Timeline.NormalizedTime)
    )
    Sprite Size=&Multiply_Vector2D(
     A=<0.5 1>
     B=&MakeVector2D(
      X=&FloatFromCurve(
       FloatCurve=DataInterface() {propertyValues {Curve {keys [<at 0 1>,<at 0.8 2 RCIM_Constant>,<at 0.95 3 RCIM_Constant>]}}}
       CurveIndex=$Particles.PerParticleRandom
      )
      Y=&FloatFromCurve(
       FloatCurve=DataInterface() {propertyValues {Curve {keys [<at 0 1>,<at 0.8 2 RCIM_Constant>,<at 0.95 3>]}}}
       CurveIndex=$Particles.PerParticleRandom
      )
     )
    )
    Sprite Size Mode=::NewEnumerator4
    UsePositionOffset=false
    Write Material Random=true
   )
   &AddVelocity(
    Cone Angle=33
    Cone Axis=<0 1 0>
    Module.UseConeFalloff=false
    Module.UseVelDistribution=false
    Override Seed=false
    Velocity Mode=::NewEnumerator2
    Velocity Speed=&Multiply_Float(
     A=&Lerp_Float(A=444, B=1111, Alpha=$Particles.PerParticleRandom)
     B=&Lerp_Float(A=-0.35, B=-1, Alpha=$System.Timeline.NormalizedTime)
    )
   )
   &ShapeLocation(MaskSpawnGroup=false, Offset=<0 -3 0>, Offset Mode=::NewEnumerator0, Override Seed=false, Sphere Radius=2, UseOwnerScale=true)
   &SkeletalMeshLocation(Bone Sampling Mode=::NewEnumerator4, MaskSpawnGroup=false, Sampled Position Offset=<0 -15 3>, Skeletal Mesh=$System.Skeletal Mesh)
  ]
  particleUpdateScript [
   &ParticleState()
   &Drag(
    Drag=&FloatFromCurve(
     FloatCurve=DataInterface() {propertyValues {Curve {keys [<at 0 0.5>,<at 0.5 0.5>,<at 1 1>]}}}
     Scale Curve=&RandomRangeFloat(Minimum=6, Maximum=22)
    )
    Use Linear Drag=true
    Use Rotational Drag=false
   )
   /Niagara/Modules/Update/Forces/CurlNoiseForce(
    Noise Frequency=0.04
    Noise Strength=&FloatFromCurve(
     FloatCurve=DataInterface() {propertyValues {Curve {keys [{tangentMode=RCTM_User, time=0, value=0}, {tangentMode=RCTM_User, time=1, value=1}]}}}
     Scale Curve=&Lerp_Float(A=4444, B=444, Alpha=$Particles.PerParticleRandom)
    )
    Pan Noise=true
    Pan Noise Field=<0 0 1>
   )
   &SolveForcesAndVelocity(Clamp Velocity=false, Limit Acceleration=false)
   &ScaleColor(
    Scale Alpha=&FloatFromCurve(
     FloatCurve=DataInterface() {propertyValues {Curve {keys [
         {interpMode=RCIM_Cubic, tangentMode=RCTM_User, time=0, value=0.6}
         {interpMode=RCIM_Cubic, tangentMode=RCTM_User, time=0.5, value=1}
         {tangentMode=RCTM_User, time=1, value=0.6}
        ]}}}
     CurveIndex=Linked() {linkedVariable {name=NPC.FortniteNPC.TimeOfDay_0-1, type=NiagaraFloat}}
    )
    Scale RGB=&MakeVectorFromLinearColorRGB(Color=&ColorFromCurve(ColorCurve=DataInterface() {propertyValues {
        AlphaCurve {keys [<at 0 1 RCIM_Cubic>]}
        BlueCurve {keys [<at 0 0 RCIM_Cubic>,<at 1 0 RCIM_Cubic>]}
        GreenCurve {keys [<at 0 9.1667 RCIM_Cubic>,<at 1 0 RCIM_Cubic>]}
        RedCurve {keys [<at 0 50 RCIM_Cubic>,<at 1 1 RCIM_Cubic>]}
       }}))
    ScaleA=true
    ScaleRGB=true
   )
   &ScaleSpriteSizeBySpeed(
    Max Scale Factor=<1.666 13>
    Min Scale Factor=<1.333 1.333>
    Sample Scale Factor By Curve=true
    Scale Factor Curve=DataInterface() {propertyValues {Curve {keys [<at 0 0>, {arriveTangent=1.4989, interpMode=RCIM_Cubic, tangentMode=RCTM_Break, time=1, value=1}]}}}
    Velocity Threshold=999
   )
   &ScaleSpriteSize(
    Uniform Curve Scale=&MakeFloatFromVector(Vector=$Engine.Owner.Scale, Channel=::NewEnumerator0)
    Uniform Curve Sprite Scale=DataInterface() {propertyValues {Curve {keys [
        {tangentMode=RCTM_User, time=0, value=1}
        {arriveTangent=-0.0405, interpMode=RCIM_Cubic, leaveTangent=-0.0405, tangentMode=RCTM_User, time=0.7009, value=1}
        {arriveTangent=-8.0089, interpMode=RCIM_Cubic, leaveTangent=-8.0089, tangentMode=RCTM_User, time=1, value=0}
       ]}}}
   )
   &ScaleSpriteSize(
    Scale Sprite Size Mode=::NewEnumerator0
    Uniform Scale Factor=&Lerp_Float(A=0.5, B=1, Alpha=$System.Timeline.NormalizedTime)
   ) {module=ScaleSpriteSize001}
  ]
  rendererClasses [&NiagaraSpriteRendererProperties]
  renderers [{rendererClass=&NiagaraSpriteRendererProperties}]
 }
 Flame_Launch {
  emitterUpdateScript [
   &EmitterState()
   &SpawnRate(
    SpawnRate=&FN_FrontEndFloatModifier(Gameplay=&Multiply_Float(A=10, B=$System.IsMoving), Front End=10)
    Use Spawn Probability=false
   )
  ]
  particleSpawnScript [
   &InitializeParticle(
    Color=<rgba 5 0.4167 0 0.33>
    Lifetime=0.3
    Mesh Scale=&Multiply_VectorByFloat(Vector=<0.3 2 2>, Float=0.4)
    Mesh Scale Mode=::NewEnumerator4
    UsePositionOffset=false
    Write Material Random=true
   )
   &InitialMeshOrientation(Mesh Axis To Orient=<0 1 0>, Mesh Orientation Mode=::NewEnumerator2, Use Rotation Vector=false)
  ]
  particleUpdateScript [
   &ParticleState()
   &SkeletalMeshLocation(Bone Sampling Mode=::NewEnumerator4, MaskSpawnGroup=false, Sampled Position Offset=<0 -15 3>, Skeletal Mesh=$System.Skeletal Mesh)
   &ScaleColor(
    Scale Alpha=&Multiply_Float(
     A=&FloatFromCurve(FloatCurve=DataInterface() {propertyValues {Curve {keys [<at 0 0 RCIM_Cubic>,<at 0.5 1>,<at 1.0011 0 RCIM_Cubic>]}}})
     B=&FloatFromCurve(
      FloatCurve=DataInterface() {propertyValues {Curve {keys [
          {interpMode=RCIM_Cubic, tangentMode=RCTM_User, time=0, value=0.6}
          {interpMode=RCIM_Cubic, tangentMode=RCTM_User, time=0.5, value=1}
          {tangentMode=RCTM_User, time=1, value=0.6}
         ]}}}
      CurveIndex=Linked() {linkedVariable {name=NPC.FortniteNPC.TimeOfDay_0-1, type=NiagaraFloat}}
     )
    )
    Scale RGB=&VectorFromCurve(VectorCurve=DataInterface() {propertyValues {
       XCurve {keys [<at 0 2 RCIM_Cubic>,<at 1 1 RCIM_Cubic>]}
       YCurve {keys [<at 0 2 RCIM_Cubic>,<at 1 1 RCIM_Cubic>]}
       ZCurve {keys [<at 0 2 RCIM_Cubic>,<at 1 1 RCIM_Cubic>]}
      }})
    ScaleA=true
    ScaleRGB=true
   )
   &DynamicMaterialParameters(
    Index 0 Param 1=&FloatFromCurve(FloatCurve=DataInterface() {propertyValues {Curve {keys [<at 0 1>,<at 1 0>]}}})
    Index 0 Param 2=5
    Index 0 Param 3=20
    Index 0 Param 4=1
    Module.Param0WriteEnabled=true
    Module.Param1WriteEnabled=true
    Module.Param2WriteEnabled=true
    Module.Param3WriteEnabled=true
   )
   &ScaleMeshSize(Scale Factor=&VectorFromCurve(VectorCurve=DataInterface() {propertyValues {
       XCurve {keys [
         <at 0 0.75 RCIM_Cubic>
         <at 0.4908 3.9579>
         {arriveTangent=4.355, interpMode=RCIM_Cubic, leaveTangent=4.355, tangentMode=RCTM_User, time=0.5, value=4}
         <at 1 0.5 RCIM_Cubic>
        ]}
       YCurve {keys [
         <at 0 1 RCIM_Cubic>
         <at 0.4908 0.9933>
         {arriveTangent=0.75, interpMode=RCIM_Cubic, leaveTangent=0.75, time=0.5, value=1}
         <at 1 1.75 RCIM_Cubic>
        ]}
       ZCurve {keys [<at -0.0014 1 RCIM_Cubic>, {arriveTangent=0.7489, interpMode=RCIM_Cubic, leaveTangent=0.7489, time=0.5, value=1}, <at 1 1.75 RCIM_Cubic>]}
      }})) {module=ScaleMeshSize001}
  ]
  rendererClasses [&NiagaraMeshRendererProperties]
  renderers [
   {rendererClass=&NiagaraMeshRendererProperties, rendererIndex=0}
   {rendererClass=&NiagaraMeshRendererProperties, rendererIndex=1}
  ]
 }
 Flash {
  emitterUpdateScript [
   &EmitterState()
   &SpawnRate(
    SpawnRate=&FN_FrontEndFloatModifier(Gameplay=&Multiply_Float(A=10, B=$System.IsMoving), Front End=10)
    Use Spawn Probability=false
   )
  ]
  particleSpawnScript [
   &InitializeParticle(
    Color=<rgba 1 0.0824 0 1>
    Lifetime Max=0.13
    Lifetime Min=0.1
    Lifetime Mode=::NewEnumerator1
    Mesh Scale=&Multiply_VectorByFloat(Vector=&RandomRangeVector(Minimum=<0.4 0.3 0.3>, Maximum=<0.5 0.5 0.5>), Float=0.3)
    Mesh Scale Mode=::NewEnumerator4
    UsePositionOffset=false
    Write Material Random=true
   )
   &InitialMeshOrientation(Mesh Axis To Orient=<0 1 0>, Mesh Orientation Mode=::NewEnumerator2, Use Rotation Vector=false)
  ]
  particleUpdateScript [
   &ParticleState()
   &SkeletalMeshLocation(Bone Sampling Mode=::NewEnumerator4, MaskSpawnGroup=false, Sampled Position Offset=<0 -15 3>, Skeletal Mesh=$System.Skeletal Mesh)
   &ScaleColor(
    Scale Alpha=&FloatFromCurve(
     FloatCurve=DataInterface() {propertyValues {Curve {keys [
         {interpMode=RCIM_Cubic, tangentMode=RCTM_User, time=0, value=0.6}
         {interpMode=RCIM_Cubic, tangentMode=RCTM_User, time=0.5, value=1}
         {tangentMode=RCTM_User, time=1, value=0.6}
        ]}}}
     CurveIndex=Linked() {linkedVariable {name=NPC.FortniteNPC.TimeOfDay_0-1, type=NiagaraFloat}}
    )
    Scale RGB=&VectorFromCurve(VectorCurve=DataInterface() {propertyValues {
       XCurve {keys [<at -0.0031 48.8>,<at 1 1>]}
       YCurve {keys [<at -0.006 20.3>,<at 1 0.8>]}
       ZCurve {keys [<at -0.0011 1.0067>,<at 0.9988 0.28>]}
      }})
    ScaleA=true
    ScaleRGB=true
   ) {module=ScaleColor001}
   &ScaleMeshSize(Scale Factor=&VectorFromCurve(VectorCurve=DataInterface() {propertyValues {
       XCurve {keys [{arriveTangent=6.3957, interpMode=RCIM_Cubic, leaveTangent=6.3957, tangentMode=RCTM_User, time=0, value=0.47}, <at 0.9748 1.29>]}
       YCurve {keys [{arriveTangent=6.1645, interpMode=RCIM_Cubic, leaveTangent=6.1645, tangentMode=RCTM_User, time=0.0035, value=0.4903}, <at 0.9786 1.29>]}
       ZCurve {keys [
         <at -0.0011 1.0067 RCIM_Cubic>
         {arriveTangent=-1.635, interpMode=RCIM_Cubic, leaveTangent=-1.635, tangentMode=RCTM_User, time=1.0016, value=0.28}
        ]}
      }})) {module=ScaleMeshSize001}
  ]
  rendererClasses [&NiagaraMeshRendererProperties]
  renderers [{rendererClass=&NiagaraMeshRendererProperties, rendererIndex=0}]
 }
 Light {
  emitterUpdateScript [
   &EmitterState(
    Inactive Response=::NewEnumerator1
    Life Cycle Mode=::NewEnumerator1
    Loop Behavior=::NewEnumerator1
    Loop Duration Mode=::NewEnumerator1
    UseLoopDelay=false
   )
   &SpawnBurst_Instantaneous(Spawn Time=1, Use Spawn Probability=false, UseLoopCountLimit=false)
  ]
  particleSpawnScript [&InitializeParticle(Color=<rgba 1 0.0824 0 1>, Sprite Size Mode=::NewEnumerator3, Uniform Sprite Size=22, UsePositionOffset=false, Write Material Random=true)]
  particleUpdateScript [
   &ParticleState(Kill Particles When Lifetime Has Elapsed=false, Let Infinitely Lived Particles Die When Emitter Deactivates=true)
   &SkeletalMeshLocation(Bone Sampling Mode=::NewEnumerator4, MaskSpawnGroup=false, Sampled Position Offset=<0 -22 3>, Skeletal Mesh=$System.Skeletal Mesh)
   &SolveForcesAndVelocity(Clamp Velocity=false, Limit Acceleration=false)
   /SetVariables(Particles.LightRadius=50) {isSetParameters=true}
   /SetVariables(Particles.Flicker:NiagaraFloat=&Waveform(
     "Waveform [1]"=::NewEnumerator0
     "[1] Frequency"=3
     "[1] Amplitude Min/Max"=&RandomRangeVector2D(Minimum=<0.6 1>, Maximum=<0.8 1>)
    )) {isSetParameters=true}
   &ScaleColor(
    Scale Alpha=&Multiply_Float(
     A=&Multiply_Float(A=15, B=$Particles.Flicker)
     B=&FN_FrontEndFloatModifier(
      Gameplay=&FloatFromCurve(
       FloatCurve=DataInterface() {propertyValues {Curve {keys [
           {interpMode=RCIM_Cubic, tangentMode=RCTM_User, time=0, value=0.7}
           {interpMode=RCIM_Cubic, tangentMode=RCTM_User, time=0.5, value=1}
           {tangentMode=RCTM_User, time=1, value=0.7}
          ]}}}
       CurveIndex=Linked() {linkedVariable {name=NPC.FortniteNPC.TimeOfDay_0-1, type=NiagaraFloat}}
       Scale Curve=$System.Timeline002.Float
      )
      Front End=&FloatFromCurve(FloatCurve=DataInterface() {propertyValues {Curve {keys [
           {arriveTangent=3.5949, interpMode=RCIM_Cubic, leaveTangent=3.5949, tangentMode=RCTM_User, time=0, value=0}
           {interpMode=RCIM_Cubic, tangentMode=RCTM_User, time=1, value=1}
          ]}}})
     )
    )
    Scale RGB=&VectorFromCurve(
     VectorCurve=DataInterface() {propertyValues {
       XCurve {keys [<at -0.0031 48.8>,<at 1 1>]}
       YCurve {keys [<at -0.006 20.3>,<at 1 0.8>]}
       ZCurve {keys [<at -0.0011 1.0067>,<at 0.9988 0.28>]}
      }}
     CurveIndex=$System.Age
     Scale Curve=<2 2 2>
    )
    ScaleA=true
    ScaleRGB=true
   )
  ]
  rendererClasses [&NiagaraLightRendererProperties]
  renderers [{rendererClass=&NiagaraLightRendererProperties, rendererIndex=0}]
 }
}
summary {
 systemName=NS_Emote_PlankCoverAge_Exhaust
 emitters [
  {emitterName=Light, rendererClasses [&NiagaraLightRendererProperties]}
  {emitterName=EmbersGPUHigh, simTarget=GPUComputeSim, rendererClasses [&NiagaraSpriteRendererProperties]}
  {emitterName=Flash, rendererClasses [&NiagaraMeshRendererProperties]}
  {emitterName=Flame_Launch, rendererClasses [&NiagaraMeshRendererProperties]}
 ]
 userVariables [
  User.bIsFrontEnd {type=NiagaraBool}
  User.bIsFrontEndPreview {type=NiagaraBool}
 ]
}
